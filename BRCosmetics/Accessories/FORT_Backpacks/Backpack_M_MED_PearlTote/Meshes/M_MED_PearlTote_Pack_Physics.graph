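# Physics asset for the PearlTote backpack cosmetic, setting up three dangling teardrop chains (left, middle, right) on the pack.
# Each chain has a base body and three dynamic segments, all capsules, with the end segment largest, alongside one large capsule for the pack.
# Segment joints have per-joint swing limits and very tight twist, while each base joint to the pack leaves swing and twist free.
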PhysicsAsset /BRCosmetics/Accessories/FORT_Backpacks/Backpack_M_MED_PearlTote/Meshes/M_MED_PearlTote_Pack_Physics
bodies {
 base_teardrop_l [{shapeType=Capsule, length=0.505, radius=0.505, shapeName=base_teardrop_l_capsule, rotation=<rot 0 -90 0>}]
 base_teardrop_m [{shapeType=Capsule, length=0.505, radius=0.505, shapeName=base_teardrop_m_capsule, rotation=<rot 0 -90 0>}]
 base_teardrop_r [{shapeType=Capsule, length=0.505, radius=0.505, shapeName=base_teardrop_r_capsule, rotation=<rot 0 -90 0>}]
 dyn_teardrop_l_01 [{shapeType=Capsule, length=1, radius=0.5, shapeName=dyn_teardrop_l_01_capsule, center=<0.5 0 0>, rotation=<rot 90 0 0>}]
 dyn_teardrop_l_02 [{shapeType=Capsule, length=1, radius=0.5, shapeName=dyn_teardrop_l_02_capsule, center=<0.8 0 0>, rotation=<rot 90 0 0>}]
 dyn_teardrop_l_03 [{shapeType=Capsule, length=4, radius=2, shapeName=dyn_teardrop_l_03_capsule, center=<3.5 0 0>, rotation=<rot 90 0 0>}]
 dyn_teardrop_m_01 [{shapeType=Capsule, length=2, radius=1, shapeName=dyn_teardrop_m_01_capsule, center=<2 0 0>, rotation=<rot 90 0 0>}]
 dyn_teardrop_m_02 [{shapeType=Capsule, length=2, radius=1, shapeName=dyn_teardrop_m_02_capsule, center=<2 0 0>, rotation=<rot 90 0 0>}]
 dyn_teardrop_m_03 [{shapeType=Capsule, length=4.5, radius=2.5, shapeName=dyn_teardrop_m_03_capsule, center=<4.5 0 0>, rotation=<rot 90 0 0>}]
 dyn_teardrop_r_01 [{shapeType=Capsule, length=1, radius=0.5, shapeName=dyn_teardrop_r_01_capsule, center=<0.5 0 0>, rotation=<rot 90 0 0>}]
 dyn_teardrop_r_02 [{shapeType=Capsule, length=1, radius=0.5, shapeName=dyn_teardrop_r_02_capsule, center=<0.8 0 0>, rotation=<rot 90 0 0>}]
 dyn_teardrop_r_03 [{shapeType=Capsule, length=3.5, radius=2, shapeName=dyn_teardrop_r_03_capsule, center=<3.5 0 0>, rotation=<rot 90 0 0>}]
 pack [{shapeType=Capsule, length=18.2812, radius=15.6808, shapeName=pack_capsule, center=<0 16.5 -20>}]
}
constraints [
 {bone1Name=dyn_teardrop_l_01, bone2Name=base_teardrop_l, swing1LimitDegrees=30, swing2LimitDegrees=5, twistLimitDegrees=1}
 {bone1Name=dyn_teardrop_l_02, bone2Name=dyn_teardrop_l_01, swing1LimitDegrees=22, swing2LimitDegrees=8, twistLimitDegrees=1}
 {bone1Name=dyn_teardrop_l_03, bone2Name=dyn_teardrop_l_02, swing1LimitDegrees=15, swing2LimitDegrees=10, twistLimitDegrees=1}
 {bone1Name=dyn_teardrop_m_01, bone2Name=base_teardrop_m, swing1LimitDegrees=40, swing2LimitDegrees=10, twistLimitDegrees=1}
 {bone1Name=dyn_teardrop_m_02, bone2Name=dyn_teardrop_m_01, swing1LimitDegrees=25, swing2LimitDegrees=15, twistLimitDegrees=1}
 {bone1Name=dyn_teardrop_m_03, bone2Name=dyn_teardrop_m_02, swing1LimitDegrees=50, swing2LimitDegrees=15, twistLimitDegrees=1}
 {bone1Name=dyn_teardrop_r_01, bone2Name=base_teardrop_r, swing1LimitDegrees=30, swing2LimitDegrees=8, twistLimitDegrees=1}
 {bone1Name=dyn_teardrop_r_02, bone2Name=dyn_teardrop_r_01, swing1LimitDegrees=25, swing2LimitDegrees=10, twistLimitDegrees=1}
 {bone1Name=dyn_teardrop_r_03, bone2Name=dyn_teardrop_r_02, swing1LimitDegrees=20, swing2LimitDegrees=15, twistLimitDegrees=1}
 {
  bone1Name=base_teardrop_l
  bone2Name=pack
  swing1LimitDegrees=45
  swing1Motion=Free
  swing2LimitDegrees=45
  swing2Motion=Free
  twistLimitDegrees=45
  twistMotion=Free
 }
 {
  bone1Name=base_teardrop_r
  bone2Name=pack
  swing1LimitDegrees=45
  swing1Motion=Free
  swing2LimitDegrees=45
  swing2Motion=Free
  twistLimitDegrees=45
  twistMotion=Free
 }
 {
  bone1Name=base_teardrop_m
  bone2Name=pack
  swing1LimitDegrees=45
  swing1Motion=Free
  swing2LimitDegrees=45
  swing2Motion=Free
  twistLimitDegrees=45
  twistMotion=Free
 }
]
