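# Material function that overlays a fuzzy, fluffy fabric look on masked areas of the PearlTote backpack material.
# Bump-offset noise layers fake fibre depth, with a platform switch choosing high-end or low-end color and alpha.
# Sun-direction shading tints the subsurface result, while Fresnel-shaped alpha feeds a temporally dithered opacity mask.
# Static switches toggle the effect and its overlay stage, which adds normal-based vertex offset and opacity masking.
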MaterialFunction /BRCosmetics/Accessories/FORT_Backpacks/Backpack_M_MED_PearlTote/Materials/MF_FluffyFabric
graph {nodes [
  Constant()
  n1 = ScalarParameter() {defaultValue=0.4, group=FluffyFabric, parameterName=FluffyFabric_WPO}
  n2 = VertexNormalWS()
  n3 = GetMaterialAttributes(None=NamedRerouteDeclaration(None=FunctionInput() {inputName=Material, inputType=FunctionInput_MaterialAttributes}) {name=MA_IN, nodeColor=<rgba 0.2422 0 1 1>}) {attributeGetTypes [#1, #2]}
  n4 = VectorParameter() {group=FluffyFabric, parameterName=FluffyFabric_SubsurfaceColorTint, defaultValue=<rgba 0.4167 0.4167 0.4167 1>}
  n5 = SetMaterialAttributes!(
   MaterialAttributes=@n3.MaterialAttributes
   BaseColor=NamedRerouteUsage()
   Metallic
   Specular=ScalarParameter() {defaultValue=0.5, group=FluffyFabric, parameterName=FluffyFabric_Specular}
   Roughness=ScalarParameter() {defaultValue=0.9, group=FluffyFabric, parameterName=FluffyFabric_Roughness}
   ShadingModel=ShadingModel() {shadingModel=MSM_Subsurface}
   SubsurfaceColor=Multiply(A=NamedRerouteUsage(), B=@n4.RGB)
  ) {attributeSetTypes [
    #1
    #3
    #4
    #5
    #6
    #7
   ]}
  n6 = BlendMaterialAttributes(A=NamedRerouteUsage(), B=@n5, Alpha=NamedRerouteUsage())
  n7 = SetMaterialAttributes!(
   MaterialAttributes=@n6
   WorldPositionOffset=LinearInterpolate(B=Multiply(A=@n1, B=@n2), Alpha=NamedRerouteUsage())
   OpacityMask=NamedRerouteUsage()
   Metallic=Constant()
  ) {attributeSetTypes [#8, #9, #3]}
  FunctionOutput(None=StaticSwitchParameter(True=StaticSwitchParameter(True=@n7, False=@n6) {group=FluffyFabric, parameterName=UseOverlay, sortPriority=0}, False=NamedRerouteUsage()) {group=FluffyFabric, parameterName=UseFluffyFabric, sortPriority=0})
  n8 = NamedRerouteUsage()
  n9 = NamedRerouteUsage()
  n10 = LinearInterpolate(
   B=Multiply(
    A=MaterialFunctionCall("Xb1,PS4"=@n8, "PC-High&&Epic,XSX,PS5"=@n8, "PC-Perf,Android"=@n9, Switch=@n9, "PC-Low(NotPerf)&&Medium"=@n9) {materialFunction=&PlatformSwitch_1D}.Result
    B=ScalarParameter() {defaultValue=4, group=FluffyFabric, parameterName=FluffyBrightness}
   ) {constB=3, desc=blend together layers and scale brightness}
   Alpha=ScalarParameter() {defaultValue=0.4, group=FluffyFabric, parameterName=FluffyNoiseContrast, sliderMax=1}
  ) {constA=1, constAlpha=0.7}
  n11 = LinearInterpolate(B=ConstantBiasScale(None=NamedRerouteUsage()), Alpha=ScalarParameter() {defaultValue=0.5, group=FluffyFabric, parameterName=NdotLIntensity}) {constA=1}
  n12 = Multiply(A=Saturate(None=Multiply(A=@n10, B=@n11)), B=NamedRerouteUsage())
  Multiply(
   A=@n12
   B=LinearInterpolate(B=NamedRerouteUsage(), Alpha=ScalarParameter() {group=FluffyFabric, parameterName=FluffyFabric_AOIntensity}) {constA=1}
  )
  n13 = TextureObjectParameter() {
   group=FluffyFabric
   parameterName=FluffyNoiseTex
   texture=/BRCosmetics/Accessories/FORT_Backpacks/Backpack_M_MED_PearlTote/Textures/T_Pearltote_Backpack_Fluffy
  }
  NamedRerouteDeclaration(None=@n13) {name=NoiseTex, nodeColor=<rgba 0.1719 1 0 1>}
  NamedRerouteDeclaration(None=ComponentMask(None=FunctionInput() {inputName=M, sortPriority=1}) {r=true}) {name=AOMap, nodeColor=<rgba 0.3828 0 1 1>}
  n14 = NamedRerouteUsage()
  n15 = NamedRerouteUsage()
  n16 = NamedRerouteUsage()
  NamedRerouteDeclaration(None=LinearInterpolate(A=LinearInterpolate(A=@n14, B=@n15, Alpha=@n16), B=NamedRerouteUsage(), Alpha=NamedRerouteUsage())) {name=Color_HighEnd, nodeColor=<rgba 0 1 0.1797 1>}
  NamedRerouteDeclaration(None=Constant() {r=-0.2}) {name=Threshold, nodeColor=<rgba 1 0 0.1406 1>}
  n17 = VectorParameter() {group=FluffyFabric, parameterName=FluffyFabric_NoiseTiling, defaultValue=<rgba 4 4 0 1>}
  NamedRerouteDeclaration(None=Multiply(A=TextureCoordinate(), B=ComponentMask(None=@n17.RGB) {g=true, r=true}) {constB=20}) {name=UVs, nodeColor=<rgba 0.1484 1 0 1>}
  NamedRerouteDeclaration(None=Transform(None=@n3.Normal)) {name=Normal_In, nodeColor=<rgba 1 0.3516 0 1>}
  NamedRerouteDeclaration(None=Constant() {r=0.1}) {name=HeightRatio, nodeColor=<rgba 1 0 0.4687 1>}
  NamedRerouteDeclaration(None=Constant()) {name=ReferencePlane, nodeColor=<rgba 0 0.7891 1 1>}
  NamedRerouteDeclaration(None=@n12) {name=Debug, nodeColor=<rgba 1 0 0.2344 1>}
  NamedRerouteDeclaration(None=ScalarParameter() {defaultValue=4, parameterName=LayerNum}) {name=LayerNum, nodeColor=<rgba 1 0 0.4219 1>}
  n18 = TextureSample(
   UVs=MaterialFunctionCall!(
    UV Coord=NamedRerouteUsage()
    Height=ScalarParameter() {defaultValue=0.75, group=FluffyFabric, parameterName=Height 01}
    ReferencePlane=NamedRerouteUsage()
    HeightRatio=NamedRerouteUsage()
    Normal
   ) {materialFunction=/Game/Global/MaterialFunctions/MF_BumpOffset_Accurate}.Result
   Tex=NamedRerouteUsage()
  ) {samplerSource=SSM_Wrap_WorldGroupSettings, samplerType=SAMPLERTYPE_Masks}
  n19 = Multiply(A=@n18.R, B=ScalarParameter() {defaultValue=0.2, group=FluffyFabric, parameterName=Alpha Multi 01, sliderMax=1})
  n20 = NamedRerouteUsage()
  n21 = ScalarParameter() {defaultValue=0.2, group=FluffyFabric, parameterName=EdgeSoftness, sliderMax=1}
  NamedRerouteDeclaration(None=MaterialFunctionCall!(In=Add(A=@n19, B=@n20), Ramp Width=@n21) {materialFunction=/Engine/Functions/Engine_MaterialFunctions01/Math/SmoothCeil}.Result) {name=Alpha 01, nodeColor=<rgba 1 0.7969 0 1>}
  n22 = TextureSample(
   UVs=MaterialFunctionCall(
    UV Coord=NamedRerouteUsage()
    Height=ScalarParameter() {defaultValue=0.5, group=FluffyFabric, parameterName=Height 02}
    ReferencePlane=NamedRerouteUsage()
    HeightRatio=NamedRerouteUsage()
   ) {materialFunction=&MF_BumpOffset_Accurate}.Result
   Tex=NamedRerouteUsage()
  ) {samplerSource=SSM_Wrap_WorldGroupSettings, samplerType=SAMPLERTYPE_Masks}
  n23 = Multiply(A=@n22.R, B=ScalarParameter() {defaultValue=0.25, group=FluffyFabric, parameterName=Alpha Multi 02, sliderMax=1})
  n24 = NamedRerouteUsage()
  NamedRerouteDeclaration(None=MaterialFunctionCall(In=Add(A=@n23, B=@n24), Ramp Width=@n21) {materialFunction=&SmoothCeil}.Result) {name=Alpha 02, nodeColor=<rgba 0.4766 0 1 1>}
  NamedRerouteDeclaration(None=@n3.BaseColor) {name=Basecolor_IN, nodeColor=<rgba 0 0.9297 1 1>}
  NamedRerouteDeclaration(None=Multiply(A=@n18.R, B=ScalarParameter() {defaultValue=0.3, group=FluffyFabric, parameterName=Color Multi 01, sliderMax=1})) {name=Color 01, nodeColor=<rgba 0 1 0.1562 1>}
  NamedRerouteDeclaration(None=Multiply(A=@n22.R, B=ScalarParameter() {defaultValue=0.35, group=FluffyFabric, parameterName=Color Multi 02, sliderMax=1})) {name=Color 02, nodeColor=<rgba 1 0 0.1641 1>}
  n25 = Custom() {code="return View.DirectionalLightDirection;", description=DirectionalLightVector, inputs []}
  n26 = CollectionParameter() {
   collection=&FortniteMaterialParameters
   parameterName=SunlightDirectionalVector
  }
  n27 = ComponentMask(None=@n26) {b=true, g=true, r=true}
  n28 = If(
   A=Custom() {code="return View.DirectionalLightColor;", description=DirectionalLightColor, inputs []}
   "A > B"=If(A=DotProduct(A=@n25, B=@n25), "A > B"=@n25, "A == B"=@n27, "A < B"=@n27) {desc=check if DirectionalLight exists}
   "A == B"=@n27
   "A < B"=@n27
  ) {desc=check if DirectionalLight color is Black}
  NamedRerouteDeclaration(None=DotProduct(A=NamedRerouteDeclaration(None=@n28) {name=LightVector, nodeColor=<rgba 1 0.4688 0 1>}, B=NamedRerouteUsage())) {name=LdotN, nodeColor=<rgba 0.2656 0 1 1>}
  NamedRerouteDeclaration(None=ChannelMaskParameter(None=FunctionInput() {inputName=TechArtMask, sortPriority=2}) {group=FluffyFabric, parameterName=FluffyChannel}.R) {name=FluffyMask, nodeColor=<rgba 1 0.75 0 1>}
  NamedRerouteDeclaration(None=@n12) {name=FluffyFabricColor, nodeColor=<rgba 1 0 0.8429 1>}
  n29 = TextureSample(
   UVs=MaterialFunctionCall(
    UV Coord=NamedRerouteUsage()
    Height=ScalarParameter() {defaultValue=0.25, group=FluffyFabric, parameterName=Height03}
    ReferencePlane=NamedRerouteUsage()
    HeightRatio=NamedRerouteUsage()
   ) {materialFunction=&MF_BumpOffset_Accurate}.Result
   Tex=NamedRerouteUsage()
  ) {samplerSource=SSM_Wrap_WorldGroupSettings, samplerType=SAMPLERTYPE_Masks}
  n30 = Add(
   A=Multiply(A=@n29.R, B=ScalarParameter() {defaultValue=0.3, group=FluffyFabric, parameterName=Alpha Multi 03, sliderMax=1})
   B=NamedRerouteUsage()
  )
  NamedRerouteDeclaration(None=NamedRerouteDeclaration(None=MaterialFunctionCall(In=@n30, Ramp Width=@n21) {materialFunction=&SmoothCeil}.Result) {name=Alpha 03, nodeColor=<rgba 0.8047 1 0 1>}) {name=Alpha_LowEnd, nodeColor=<rgba 0 0.625 1 1>}
  n31 = Add(A=Add(A=NamedRerouteUsage(), B=NamedRerouteUsage()), B=NamedRerouteUsage())
  NamedRerouteDeclaration(None=@n31) {name=Alpha_HighEnd, nodeColor=<rgba 0 0.2133 1 1>}
  n32 = Power(
   Base=OneMinus(None=Fresnel() {baseReflectFraction=0, exponent=1})
   Exp=ScalarParameter() {defaultValue=2, group=FluffyFabric, parameterName=FresnelPower}
  ) {constExponent=4, desc="make core more visable to hide AO "}
  n33 = NamedRerouteUsage()
  n34 = NamedRerouteUsage()
  n35 = MaterialFunctionCall("Xb1,PS4"=@n33, "PC-High&&Epic,XSX,PS5"=@n33, "PC-Perf,Android"=@n34, Switch=@n34, "PC-Low(NotPerf)&&Medium"=@n34) {materialFunction=&PlatformSwitch_1D}
  n36 = Saturate(None=Subtract(A=Add(A=@n35.Result, B=@n32), B=Fresnel() {baseReflectFraction=0, desc=to make the edge more fuzzy, exponent=4})) {desc=opacity}
  NamedRerouteDeclaration(None=Multiply(
    A=NamedRerouteUsage()
    B=MaterialFunctionCall!(Alpha Threshold=@n36, Random) {materialFunction=/Game/Environments/Asteria/Sets/NeonCity/Techart/Screen/MaterialFunctions/DitherTemporalAABlue}.Result
   )) {name=Opacity, nodeColor=<rgba 0.125 1 0 1>}
  n37 = ScalarParameter() {defaultValue=0.4, group=FluffyFabric, parameterName=Color Multi 03}
  NamedRerouteDeclaration(None=NamedRerouteDeclaration(None=Multiply(A=@n29.R, B=@n37) {constB=0.2}) {name=Color 03, nodeColor=<rgba 0 0.4844 1 1>}) {name=Color_LowEnd, nodeColor=<rgba 0.5 0 1 1>}
  NamedRerouteUsage()
  Saturate(None=@n31)
 ]}
